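# Unlit translucent two-sided material for the JollyTroll skydiving contrail ribbon in Niagara.
# Animated sine and cosine waves, plus a panning normal map outside mobile, wiggle the ribbon texture's V coordinate, driven by Niagara dynamic parameters.
# Texture red and green tint two colors, blue adds on top, and particle color scales the emissive.
# Opacity combines alpha, a sine mask along the length, depth fade and camera fade, with optional SplineThicken offset.
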Material /BRCosmetics/Athena/Effects/Particles/Skydiving/NS_Contrail_JollyTroll/M_Skydive_JollyTrollRibbon
graph {
 properties {
  bUsedWithNiagaraMeshParticles=true
  bUsedWithNiagaraRibbons=true
  bUsedWithNiagaraSprites=true
  blendMode=BLEND_Translucent
  pixelDepthOffsetMode=PDOM_Legacy
  shadingModel=MSM_Unlit
  twoSided=true
  usedShadingModels=Unlit
 }
 nodes [
  Divide(A=ComponentMask(None=WorldPosition().XYZ) {b=true}, B=Divide(A=Constant() {r=-1618}) {constB=4}) {constB=1618}
  n1 = VectorParameter() {parameterName=UV Control, channelNames {a=Offset V, b=Offset U, g=Tiling V, r=Tiling U}, defaultValue=<rgba 1 1 0 0>}
  n2 = AppendVector(A=@n1.Offset U, B=@n1.Offset U)
  n3 = TextureCoordinate()
  n4 = Add(A=Multiply(A=AppendVector(A=@n1.Tiling U, B=@n1.Tiling U), B=Add(A=@n3) {constB=-0.5})) {constB=0.5}
  n5 = DynamicParameter() {defaultValue=<rgba 1 0.56 0 0.1>, paramNames ["offset(0-1)", speed, Time, SeaweedWiggle]}
  n6 = Add(A=@n3, B=AppendVector(A=@n5."offset(0-1)", B=Constant()))
  n7 = Add(A=Multiply(A=NamedRerouteUsage(), B=@n5.speed) {constB=10})
  n8 = Panner(Coordinate=@n6, Time=@n7) {speedX=-0.4}
  n9 = ScalarParameter() {defaultValue=1, group=Distort_map, parameterName=Disort_frequency, sortPriority=4}
  n10 = Multiply(A=Frac(None=ComponentMask(None=@n8) {r=true}), B=@n9)
  n11 = Panner(Coordinate=Add(A=@n6, B=Constant2Vector() {r=0.5}), Time=@n7) {speedX=-0.4}
  n12 = Multiply(A=Frac(None=ComponentMask(None=@n11) {r=true}), B=@n9)
  n13 = Add(A=Multiply(A=Sine(None=@n10) {period=0.5}), B=Multiply(A=Cosine(None=@n12)) {constB=0.1})
  n14 = Reroute()
  n15 = Constant2Vector() {g=0.25, r=1}
  n16 = TextureSampleParameter2D(UVs=Panner(Coordinate=Multiply(A=@n14, B=@n15), Time=NamedRerouteUsage()) {speedX=2, speedY=0.1}) {
   group=Distort_map
   parameterName=Detail_disort_map
   samplerType=SAMPLERTYPE_Normal
   texture=/Game/Global/Textures/Default/Tileable_Patterns/1Normal
  }
  n17 = ComponentMask(None=@n16.RGB) {g=true, r=true}
  n18 = Constant2Vector() {g=0.75, r=0.5}
  n19 = ShadingPathSwitch(Default=Add(A=@n13, B=Multiply(A=@n17, B=@n18)) {constB=0}, Mobile=@n13) {desc=Migrated from FeatureLevelSwitch}
  n20 = Power(Base=ComponentMask(None=TextureCoordinate()) {r=true})
  n21 = ScalarParameter() {defaultValue=2, parameterName=wiggle}
  n22 = Multiply(
   A=Multiply(A=Multiply(B=@n19) {constA=0.5}, B=ScalarParameter() {defaultValue=0.1, group=Distort_map, parameterName=Disort_amplitude, sortPriority=5}) {constB=0.15}
   B=Saturate(None=Multiply(A=@n20, B=@n21))
  )
  n23 = Constant2Vector() {g=1}
  n24 = TextureSampleParameter2D(UVs=Add(A=Add(A=@n2, B=@n4) {constB=-0.5}, B=Multiply(A=@n22, B=@n23))) {
   parameterName=Param
   samplerSource=SSM_Wrap_WorldGroupSettings
   texture=&/T_RibbonPresent_Color_Contrail
  }
  n25 = Add(
   A=Multiply(A=@n24.R, B=VectorParameter() {parameterName=Color1, defaultValue=<rgba 0.8229 0.024 0 0>}.RGB)
   B=Multiply(A=@n24.G, B=VectorParameter() {parameterName=Color2, defaultValue=<rgba 1 0.5733 0.2 0>}.RGB)
  )
  n26 = ParticleColor()
  MaterialFunctionCall(Emissive=Multiply(A=Add(A=@n25, B=@n24.B), B=Multiply(A=@n26.RGB) {constB=0.4})) {materialFunction=&MF_FX_EmissiveAdjustments}
  Multiply(
   A=MaterialFunctionCall(
    WidthBase=ScalarParameter() {defaultValue=6, group=WPO, parameterName="WidthBase (S)"}
    WidthTip=ScalarParameter() {defaultValue=20, group=WPO, parameterName="WidthTip (S)"}
   ) {materialFunction=&SplineThicken}.WorldPositionOffset
   B=ScalarParameter() {group=WPO, parameterName=UseSplineThicken}
  )
  n27 = ComponentMask(None=TextureCoordinate()) {r=true}
  n28 = Saturate(None=Multiply(A=Sine(None=@n27) {period=2}) {constB=3})
  n29 = DepthFade(FadeDistance=ScalarParameter() {defaultValue=50, group=Opacity, parameterName=DepthFade})
  n30 = MaterialFunctionCall(
   Fade Length=ScalarParameter() {defaultValue=35, group=Opacity, parameterName=Camera Fade Length}
   Fade Offset=ScalarParameter() {defaultValue=25, group=Opacity, parameterName=Camera Fade Offset}
  ) {materialFunction=&CameraDepthFade}
  Multiply(
   A=Multiply(A=@n26.A, B=Multiply(A=@n28, B=@n24.A))
   B=LinearInterpolate(A=@n29, B=Multiply(A=@n29, B=@n30.Result), Alpha=ScalarParameter() {defaultValue=1, group=Opacity, parameterName=Use Camera Fade})
  )
  NamedRerouteDeclaration(None=@n5.Time) {name=NiagaraTime, nodeColor=<rgba 0.125 1 0 1>}
  OneMinus(None=ComponentMask(None=@n3) {r=true})
 ]
}
