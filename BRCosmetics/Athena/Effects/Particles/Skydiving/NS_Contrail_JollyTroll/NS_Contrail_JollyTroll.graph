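# Christmas-themed skydiving contrail cosmetic that attaches gift meshes, glows, decoration sprites, sparkles and trails to player skeleton bones.
# Emitters sample the user skeletal mesh per partition or spawn group, with System.DiveVelocity driving trail orientation and particle offsets.
# Two mirrored ribbon emitters reposition fixed burst particles each frame by normalized index curves, with glide lean speeding a material counter.
# BowRibbons and both spawn-rate PresentRibbon emitters are disabled.
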NiagaraSystem /BRCosmetics/Athena/Effects/Particles/Skydiving/NS_Contrail_JollyTroll/NS_Contrail_JollyTroll
dependencies {
 usedDataInterfaces [
  &NiagaraDataInterfaceCamera
  &NiagaraDataInterfaceCurve
  &NiagaraDataInterfaceArrayFloat3
  &NiagaraDataInterfaceArrayFloat
  &NiagaraDataInterfaceSpriteRendererInfo
  &NiagaraDataInterfaceColorCurve
 ]
 usedDynamicInputs [
  &SelectFloatFromArray
  &SetBoolByFloatComparison
  &self:Abs_Z_Axis
  &FN_FrontEndFloatModifier
 ]
 usedModules [
  &SystemState
  &self:SystemSpawnScript.SystemScriptSource.SystemScriptGraph.NiagaraNodeAssignment_0.SetVariables_#1
  &Timeline
  &self:SystemSpawnScript.SystemScriptSource.SystemScriptGraph.NiagaraNodeAssignment_2.SetVariables_#2
  &self:SystemSpawnScript.SystemScriptSource.SystemScriptGraph.NiagaraNodeAssignment_1.SetVariables_#3
 ]
}
emitters {
 BowRibbons {
  bEnabled=false
  emitterUpdateScript [
   &EmitterState(Life Cycle Mode=::NewEnumerator1, Loop Behavior=::NewEnumerator1, Loop Duration Mode=::NewEnumerator1, UseLoopDelay=false)
   &SpawnBurst_Instantaneous(Spawn Count=2, Use Spawn Probability=false, UseLoopCountLimit=false)
   &SpawnBurst_Instantaneous(Spawn Count=2, Spawn Group=1, Use Spawn Probability=false, UseLoopCountLimit=false) {module=SpawnBurst_Instantaneous001}
  ]
  particleSpawnScript [
   &PartitionParticles()
   &InitializeParticle(Mesh Scale Mode=::NewEnumerator3, UsePositionOffset=false, Write Material Random=true) {module=InitializeParticle001}
   &InitialMeshOrientation(
    Mesh Orientation Mode=::NewEnumerator1
    Rotation=&MaskVectorBySpawnGroup(Masked Vector=<0 0.1 0>, Passthrough Vector=<0 -0.1 0>)
    Use Rotation Vector=true
   )
  ]
  particleUpdateScript [
   &ParticleState(
    Kill Particles When Lifetime Has Elapsed=false
    Let Infinitely Lived Particles Die When Emitter Deactivates=true
    Loop Particles Lifetime=true
   )
   &SkeletalMeshLocation(
    Bone / Socket Index=$Particles.Initial.PartitionParticles.Partition
    Bone Sampling Mode=::NewEnumerator4
    Inherit Velocity=false
    MaskSpawnGroup=false
    Sampled Position Offset Coordinate Space:ENiagaraMeshTransforms=::NewEnumerator1
    Skeletal Mesh=$User.SkeletalMesh2
   )
   &Drag(Drag=0.3, Use Linear Drag=true, Use Rotational Drag=false)
   /Niagara/Modules/Update/Forces/CurlNoiseForce(
    Noise Frequency:NiagaraFloat=&UniformRangedFloat(Minimum=0.1, Maximum=1)
    Noise Strength=&UniformRangedFloat(Minimum=300, Maximum=600)
    Pan Noise=false
    Pan Noise Field=<1 1 0>
   ) {enabled=false}
   &DynamicMaterialParameters(
    Index 0 Param 1=&FloatFromCurve(FloatCurve=DataInterface() {propertyValues {Curve {keys [<at -0.0046 2>,<at 1.0011 0.002>]}}})
    Module.Param0WriteEnabled=true
    Module.Param1WriteEnabled=true
    Module.Param2WriteEnabled=true
    Module.Param3WriteEnabled=true
   ) {enabled=false}
   &ScaleMeshSize(Scale Factor=&Multiply_Vector(
     A=<1 1 1>
     B=&VectorFromFloat(Value=&FloatFromCurve(FloatCurve=DataInterface() {propertyValues {Curve {keys [
           <at -0.0046 1>
           <at 0.3771 0.7748>
           {arriveTangent=-0.3128, leaveTangent=-0.3128, time=0.4459, value=1.2326}
           <at 0.5037 0.6933>
           <at 0.5587 0.6619>
           <at 0.611 0.8501>
           <at 0.6536 0.618>
           <at 0.7155 0.6055>
           <at 0.8338 0.5741>
           <at 1.0011 0.5>
          ]}}}))
    )) {enabled=false}
   &SolveForcesAndVelocity(Clamp Velocity=false, Limit Acceleration=false)
   &GenerateLocationEvent(Event Send Rate:NiagaraFloat=&UniformRangedFloat(Minimum=1, Maximum=3), Module.Use Event Send Rate=true) {enabled=false}
  ]
  rendererClasses [&NiagaraMeshRendererProperties]
  renderers [{rendererClass=&NiagaraMeshRendererProperties, rendererIndex=0}]
 }
 BoxGlow {
  emitterUpdateScript [
   &EmitterState(Life Cycle Mode=::NewEnumerator1, Loop Behavior=::NewEnumerator1, Loop Duration Mode=::NewEnumerator1, UseLoopDelay=false)
   &SpawnBurst_Instantaneous(Spawn Count=4, Use Spawn Probability=false, UseLoopCountLimit=false)
  ]
  particleSpawnScript [
   &PartitionParticles()
   &InitializeParticle(
    Color=<rgba 1 0.8423 0.1211 0.25>
    Lifetime=5
    Sprite Size Mode=::NewEnumerator3
    Uniform Sprite Size=100
    UsePositionOffset=false
    Write Material Random=true
   )
  ]
  particleUpdateScript [
   &ParticleState(
    Kill Particles When Lifetime Has Elapsed=false
    Let Infinitely Lived Particles Die When Emitter Deactivates=true
    Loop Particles Lifetime=true
   )
   &SkeletalMeshLocation(
    Bone / Socket Index=$Particles.PartitionParticles.Partition
    Bone Sampling Mode=::NewEnumerator4
    MaskSpawnGroup=false
    Sampled Position Offset=<0 0 25>
    Sampled Position Offset Coordinate Space:ENiagaraSkelMeshTransforms=::NewEnumerator1
    Skeletal Mesh=$User.SkeletalMesh
   )
  ]
  rendererClasses [&NiagaraSpriteRendererProperties]
  renderers [{rendererClass=&NiagaraSpriteRendererProperties}]
 }
 BoxMesh {
  emitterUpdateScript [
   &EmitterState(Life Cycle Mode=::NewEnumerator1, Loop Behavior=::NewEnumerator1, Loop Duration Mode=::NewEnumerator1, UseLoopDelay=false)
   &SpawnBurst_Instantaneous(Spawn Count=2, Spawn Group=1, Use Spawn Probability=false, UseLoopCountLimit=false)
  ]
  particleSpawnScript [
   &FN_SetSpawnGroup(Number of Spawn Groups=2)
   &InitializeParticle(Mesh Scale Mode=::NewEnumerator3, UsePositionOffset=false, Write Material Random=true) {module=InitializeParticle001}
   &CylinderLocation(
    Cylinder Height=1
    Cylinder Radius=1
    Lathe Profile=DataInterface() {propertyValues {Curve {keys []}}}
    MaskSpawnGroup=false
    Module.Enable Lathe Profile=false
    Module.Override Randomness=false
    Module.Surface Only=false
   ) {enabled=false}
   &AddVelocity(Velocity=&UniformRangedVector(
     Minimum=&MakeVector(
      X=0
      Y=0
      Z=&Multiply_Float(
       A=150
       B=&FloatFromCurve(
        FloatCurve=DataInterface() {propertyValues {Curve {keys [<at 0 0.9>,<at 1 2>]}}}
        CurveIndex=&Divide_Float(A=&VectorLength(Vector=$Engine.Owner.Velocity), B=9000)
       )
      )
     )
     Maximum=&MakeVector(
      X=0
      Y=0
      Z=&Multiply_Float(
       A=200
       B=&FloatFromCurve(
        FloatCurve=DataInterface() {propertyValues {Curve {keys [<at 0 0.9>,<at 1 2>]}}}
        CurveIndex=&Divide_Float(A=&VectorLength(Vector=$Engine.Owner.Velocity), B=9000)
       )
      )
     )
    )) {enabled=false}
   /SetVariables(Particles.RotationOrientation:NiagaraFloat=&UniformRangedFloat(Minimum=-1, Maximum=1)) {enabled=false, isSetParameters=true, module=SetVariables_#4}
   /SetVariables(Particles.RotationSpeed:Vector3f=&UniformRangedVector(Minimum=<0 0 -0.5>, Maximum=<0 0 0.5>)) {enabled=false, isSetParameters=true, module=SetVariables_#5}
   &InitialMeshOrientation(Mesh Axis To Orient=<0 0 0>, Mesh Orientation Mode=::NewEnumerator2, Orientation Vector=<0 0 0>, Use Rotation Vector=false)
  ]
  particleUpdateScript [
   &ParticleState(
    Kill Particles When Lifetime Has Elapsed=false
    Let Infinitely Lived Particles Die When Emitter Deactivates=true
    Loop Particles Lifetime=true
   )
   &UpdateMeshOrientation(
    Facing Direction=$System.DiveVelocity
    Facing Interpolation=false
    Orientation Method=::NewEnumerator1
    Override Quat=false
    TransformFacing=false
    TransformUp=false
    Up Direction:Vector3f=<0 0 0>
   ) {enabled=false, module=UpdateMeshOrientation002}
   &SkeletalMeshLocation(
    Bone / Socket Index=$Particles.SpawnGroup
    Bone Sampling Mode=::NewEnumerator4
    Inherit Velocity=false
    MaskSpawnGroup=false
    Sampled Position Offset=<0 -6 10>
    Sampled Position Offset Coordinate Space:ENiagaraMeshTransforms=::NewEnumerator1
    Skeletal Mesh=$User.SkeletalMesh
   )
   /Niagara/Modules/Update/Orientation/V2/UpdateMeshOrientation(Rotation Rate:NiagaraFloat=$Particles.RotationOrientation, Rotation Vector:Vector3f=$Particles.RotationSpeed) {enabled=false}
   &Drag(Drag=0.3, Use Linear Drag=true, Use Rotational Drag=false)
   &CurlNoiseForce(
    Noise Frequency:NiagaraFloat=&UniformRangedFloat(Minimum=0.1, Maximum=1)
    Noise Strength=&UniformRangedFloat(Minimum=300, Maximum=600)
    Pan Noise=false
    Pan Noise Field=<1 1 0>
   ) {enabled=false}
   &DynamicMaterialParameters(
    Index 0 Param 1=&FloatFromCurve(FloatCurve=DataInterface() {propertyValues {Curve {keys [<at -0.0046 2>,<at 1.0011 0.002>]}}})
    Module.Param0WriteEnabled=true
    Module.Param1WriteEnabled=true
    Module.Param2WriteEnabled=true
    Module.Param3WriteEnabled=true
   ) {enabled=false}
   &ScaleMeshSize(Scale Factor=&Multiply_Vector(
     A=<1 1 1>
     B=&VectorFromFloat(Value=&FloatFromCurve(FloatCurve=DataInterface() {propertyValues {Curve {keys [
           <at -0.0046 1>
           <at 0.3771 0.7748>
           {arriveTangent=-0.3128, leaveTangent=-0.3128, time=0.4459, value=1.2326}
           <at 0.5037 0.6933>
           <at 0.5587 0.6619>
           <at 0.611 0.8501>
           <at 0.6536 0.618>
           <at 0.7155 0.6055>
           <at 0.8338 0.5741>
           <at 1.0011 0.5>
          ]}}}))
    )) {enabled=false}
   &SolveForcesAndVelocity(Clamp Velocity=false, Limit Acceleration=false)
   &GenerateLocationEvent(Event Send Rate:NiagaraFloat=&UniformRangedFloat(Minimum=1, Maximum=3), Module.Use Event Send Rate=true) {enabled=false}
  ]
  rendererClasses [&NiagaraMeshRendererProperties]
  renderers [{rendererClass=&NiagaraMeshRendererProperties, rendererIndex=0}]
 }
 ChristmasDecorationsHands {
  emitterUpdateScript [
   &EmitterState(Delay First Loop Only=true, Life Cycle Mode=::NewEnumerator1, Loop Delay=0.25, Loop Duration=1, UseLoopDelay=true)
   &SpawnRate(SpawnRate=25, Use Spawn Probability=true)
  ]
  particleSpawnScript [
   &PartitionParticles(Number Of Partitions=4)
   &InitializeParticle(
    Lifetime Max=0.75
    Lifetime Min=0.5
    Lifetime Mode=::NewEnumerator1
    Sprite Rotation Mode=::NewEnumerator2
    Sprite Size Mode=::NewEnumerator1
    Uniform Sprite Size Max=15
    Uniform Sprite Size Min=10
    UsePositionOffset=false
    Write Material Random=true
   ) {module=InitializeParticle001}
   &SubUVAnimation(
    "Sprite Renderer "=DataInterface() {propertyValues {SpriteRenderer=&self:Gifts_1.NiagaraSpriteRendererProperties_0}}
    SubUV Animation Mode:ENiagara_SubUVLookupModeV2=::NewEnumerator2
    UseEndFrame=false
    UseStartFrame=false
   )
  ]
  particleUpdateScript [
   &ParticleState()
   &SpriteRotationRate(Rotation Rate=&FloatFromCurve(
     FloatCurve=DataInterface() {propertyValues {Curve {keys [
         {arriveTangent=4.1785, interpMode=RCIM_Cubic, leaveTangent=4.1785, tangentMode=RCTM_User, time=0, value=0}
         {arriveTangent=1.1179, interpMode=RCIM_Cubic, leaveTangent=1.1179, tangentMode=RCTM_User, time=0.2042, value=0.6228}
         {arriveTangent=0.0224, interpMode=RCIM_Cubic, leaveTangent=0.0224, tangentMode=RCTM_User, time=1, value=1}
        ]}}}
     Scale Curve=150
    ))
   &SkeletalMeshLocation(
    Bone / Socket Index=$Particles.PartitionParticles.Partition
    Bone Sampling Mode=::NewEnumerator4
    MaskSpawnGroup=false
    Orientation Sampling=::NewEnumerator1
    Sampled Position Offset=&MakeCustomVectorFromBool(
     Bool=&SetBoolByIntComparison(A=$Particles.PartitionParticles.Partition, B=1, Comparison Type=::NewEnumerator0)
     True Vector=<0 0 0>
     False Vector=<1 0 25>
    )
    Sampled Position Offset Coordinate Space:ENiagaraSkelMeshTransforms=::NewEnumerator1
    Skeletal Mesh=$User.SkeletalMesh
   ) {module=SkeletalMeshLocation001}
   &ShapeLocation(
    Box Size=<15 15 1>
    MaskSpawnGroup=false
    Offset Coordinate Space=::World
    Offset Mode=::NewEnumerator0
    Override Seed=false
    Shape Origin=&ConvertPositionToVector(Input Position=$Particles.Position, Convert To Absolute Large World Vector=false)
    Shape Primitive=::NewEnumerator2
    Surface Only=false
    UseOwnerScale=true
   )
   /SetVariables(Particles.Position=&Add_Vector(
     A=&ConvertPositionToVector(Input Position=$Particles.Position, Convert To Absolute Large World Vector=false)
     B=&Multiply_VectorByFloat(
      Vector=$System.DiveVelocity
      Float=&Add_Float(
       A=&FloatFromCurve(
        FloatCurve=DataInterface() {propertyValues {Curve {keys [
            {interpMode=RCIM_Cubic, tangentMode=RCTM_User, time=0, value=0}
            {arriveTangent=1.6542, interpMode=RCIM_Cubic, leaveTangent=1.6542, tangentMode=RCTM_User, time=1, value=1}
           ]}}}
        Scale Curve=&Multiply_Float(A=$System.DiveOffset, B=&RandomRangeFloat(Minimum=0.25, Maximum=0.25))
       )
       B=1
      )
     )
    )) {isSetParameters=true}
   &SolveForcesAndVelocity(Clamp Velocity=false, Limit Acceleration=false)
   &ScaleSpriteSize(Uniform Curve Sprite Scale=DataInterface() {propertyValues {Curve {keys [
        {interpMode=RCIM_Cubic, tangentMode=RCTM_User, time=0, value=0}
        {interpMode=RCIM_Cubic, tangentMode=RCTM_User, time=0.1, value=1.175}
        {arriveTangent=-1.3056, interpMode=RCIM_Cubic, leaveTangent=-1.3056, time=0.8, value=1.25}
        {tangentMode=RCTM_User, time=1, value=0}
       ]}}})
   &ScaleColor(
    Scale Alpha=&FloatFromCurve(FloatCurve=DataInterface() {propertyValues {Curve {keys [{interpMode=RCIM_Cubic, tangentMode=RCTM_User, time=0.65, value=1}, {interpMode=RCIM_Cubic, tangentMode=RCTM_User, time=1, value=0}]}}})
    ScaleA=true
    ScaleRGB=true
   )
   &CameraOffset(Camera Offset Amount=2)
  ]
  rendererClasses [&NiagaraSpriteRendererProperties]
  renderers [{rendererClass=&NiagaraSpriteRendererProperties}]
 }
 Empty {
  emitterUpdateScript [
   &EmitterState()
   &SpawnBurst_Instantaneous(Spawn Count=&Multiply_Int(A=10, B=4), Use Spawn Probability=false, UseLoopCountLimit=true)
  ]
  particleSpawnScript [
   &PartitionParticles(Number Of Partitions=4, Write Partition to Ribbon ID=::NewEnumerator0)
   /SetVariables(Particles.RibbonLinkOrder:NiagaraFloat=&ReturnNormalizedExecIndex()) {isSetParameters=true, module=SetVariables_#6}
   &InitializeParticle(
    Color Mode=::NewEnumerator0
    Lifetime=5
    Position Offset=<0 0 25>
    Ribbon Width=25
    Ribbon Width Mode=::NewEnumerator1
    Sprite Size Mode=::NewEnumerator3
    Uniform Sprite Size=5
    UsePositionOffset=false
    Write Material Random=true
   )
   &SkeletalMeshLocation(
    Bone / Socket Index=$Particles.PartitionParticles.Partition
    Bone Sampling Mode=::NewEnumerator4
    MaskSpawnGroup=false
    Sampled Position Offset Coordinate Space:ENiagaraSkelMeshTransforms=::NewEnumerator2
    Skeletal Mesh=$User.SkeletalMesh
   )
  ]
  particleUpdateScript [
   &ParticleState(
    Kill Particles When Lifetime Has Elapsed=false
    Let Infinitely Lived Particles Die When Emitter Deactivates=true
    Loop Particles Lifetime=true
   )
   /Niagara/Modules/Update/Utility/IncrementOverTime(Rate of Change:NiagaraFloat=&Lerp_Float(A=1, B=1.5, Alpha=$User.ParaGlideLeanAlpha))
   &DynamicMaterialParameters(
    Index 0 Param 1=&Multiply_FloatByInt(
     Float=0.5
     Integer=&ReturnParticleIDAttributes(Particle ID Attribute=::NewEnumerator2, Niagara ID=$Particles.RibbonID)
    )
    Index 0 Param 2=6
    Index 0 Param 3=$Particles.IncrementOverTime.Counter
    Module.Param0WriteEnabled=true
    Module.Param1WriteEnabled=true
    Module.Param2WriteEnabled=true
    Module.Param3WriteEnabled=true
   )
   /Niagara/Modules/Update/Forces/V2/CurlNoiseForce(
    Noise Frequency=25
    Noise Strength=&Multiply_Float(A=1, B=$Particles.RibbonLinkOrder)
    Pan Noise=true
    Pan Noise Field=<0 0 2>
   )
   &SkeletalMeshLocation(
    Bone / Socket Index=$Particles.PartitionParticles.Partition
    Bone Sampling Mode=::NewEnumerator4
    MaskSpawnGroup=false
    Sampled Position Offset=&MakeCustomVectorFromBool(
     Bool=&SetBoolByIntComparison(A=$Particles.PartitionParticles.Partition, B=1, Comparison Type=::NewEnumerator0)
     True Vector=<0 0 0>
     False Vector=<0 0 17>
    )
    Sampled Position Offset Coordinate Space:ENiagaraSkelMeshTransforms=::NewEnumerator1
    Skeletal Mesh=$User.SkeletalMesh
   ) {module=SkeletalMeshLocation001}
   /SetVariables(Particles.Position=&AddVectorToPosition(
     Position=$Particles.Position
     Vector=&Multiply_VectorByFloat(
      Vector=$System.DiveVelocity
      Float=&FloatFromCurve(
       FloatCurve=DataInterface() {propertyValues {Curve {keys [{tangentMode=RCTM_User, time=0, value=0}, {tangentMode=RCTM_User, time=1, value=1}]}}}
       CurveIndex=&ReturnNormalizedExecIndex()
       Scale Curve=&MakeCustomFloatFromBool(
        Bool=&SetBoolByIntComparison(A=$Particles.PartitionParticles.Partition, B=1, Comparison Type=::NewEnumerator0)
        True Float=-40
        False Float=-85
       )
      )
     )
    )) {isSetParameters=true, module=SetVariables_#7}
   /SetVariables(Particles.Position=&AddVectorToPosition(
     Position=$Particles.Position
     Vector=&TransformVector(
      Vector=&Multiply_VectorByFloat(
       Vector=&SelectVectorFromArray(
        Direct Array Index=$Particles.PartitionParticles.Partition
        Vector Selection Array=DataInterface() {propertyValues {FloatData [<0.75 0 0>, <-0.75 0 0>, <0.75 0 0>, <-0.75 0 0>]}}
       )
       Float=&FloatFromCurve(
        FloatCurve=DataInterface() {propertyValues {Curve {keys [
            {arriveTangent=4.1785, interpMode=RCIM_Cubic, leaveTangent=4.1785, tangentMode=RCTM_User, time=0, value=0}
            {arriveTangent=1.1179, interpMode=RCIM_Cubic, leaveTangent=1.1179, tangentMode=RCTM_User, time=0.2042, value=0.6228}
            {arriveTangent=0.0224, interpMode=RCIM_Cubic, leaveTangent=0.0224, tangentMode=RCTM_User, time=1, value=1}
           ]}}}
        CurveIndex=&ReturnNormalizedExecIndex()
        Scale Curve=&MakeCustomFloatFromBool(
         Bool=&SetBoolByIntComparison(A=$Particles.PartitionParticles.Partition, B=1, Comparison Type=::NewEnumerator0)
         True Float=6
         False Float=13
        )
       )
      )
      Destination Space=::Simulation
     )
    )) {isSetParameters=true, module=SetVariables_#8}
   /SetVariables(Particles.Position=&AddVectorToPosition(
     Position=$Particles.Position
     Vector=&Multiply_VectorByFloat(
      Vector=&NormalizeVector(Vector To Normalize=$Output.CurlNoiseForce.CurlNoiseForce, Safe Normalize=true, Safe Fallback Threshold=0, Safe Fallback Vector=<0 0 0.001>)
      Float=&FloatFromCurve(
       FloatCurve=DataInterface() {propertyValues {Curve {keys [{tangentMode=RCTM_User, time=0, value=0}, {tangentMode=RCTM_User, time=1, value=1}]}}}
       CurveIndex=&ReturnNormalizedExecIndex()
       Scale Curve=&Lerp_Float(A=1, B=2, Alpha=$User.ParaGlideLeanAlpha)
      )
     )
    )) {enabled=false, isSetParameters=true, module=SetVariables_#9}
   &ScaleRibbonWidth(Ribbon Width Scale=1.15)
  ]
  rendererClasses [&NiagaraSpriteRendererProperties, &NiagaraRibbonRendererProperties]
  renderers [
   {rendererClass=&NiagaraSpriteRendererProperties}
   {rendererClass=&NiagaraRibbonRendererProperties, rendererIndex=1}
  ]
 }
 Empty001 {
  emitterUpdateScript [
   &EmitterState()
   &SpawnBurst_Instantaneous(Spawn Count=&Multiply_Int(A=10, B=4), Use Spawn Probability=false, UseLoopCountLimit=true)
  ]
  particleSpawnScript [
   &PartitionParticles(Number Of Partitions=4, Write Partition to Ribbon ID=::NewEnumerator0)
   /SetVariables(Particles.RibbonLinkOrder:NiagaraFloat=&ReturnNormalizedExecIndex()) {isSetParameters=true, module=SetVariables_#6}
   &InitializeParticle(
    Color Mode=::NewEnumerator0
    Lifetime=5
    Position Offset=<0 0 25>
    Ribbon Width=25
    Ribbon Width Mode=::NewEnumerator1
    Sprite Size Mode=::NewEnumerator3
    Uniform Sprite Size=5
    UsePositionOffset=false
    Write Material Random=true
   )
   &SkeletalMeshLocation(
    Bone / Socket Index=$Particles.PartitionParticles.Partition
    Bone Sampling Mode=::NewEnumerator4
    MaskSpawnGroup=false
    Sampled Position Offset Coordinate Space:ENiagaraSkelMeshTransforms=::NewEnumerator2
    Skeletal Mesh=$User.SkeletalMesh
   )
  ]
  particleUpdateScript [
   &ParticleState(
    Kill Particles When Lifetime Has Elapsed=false
    Let Infinitely Lived Particles Die When Emitter Deactivates=true
    Loop Particles Lifetime=true
   )
   &ScaleSpriteSize()
   &IncrementOverTime(Rate of Change:NiagaraFloat=&Lerp_Float(A=1, B=1.5, Alpha=$User.ParaGlideLeanAlpha))
   &DynamicMaterialParameters(
    Index 0 Param 2=10
    Index 0 Param 3=$Particles.IncrementOverTime.Counter
    Module.Param0WriteEnabled=true
    Module.Param1WriteEnabled=true
    Module.Param2WriteEnabled=true
    Module.Param3WriteEnabled=true
   )
   &CurlNoiseForce(
    Noise Frequency=25
    Noise Strength=&Multiply_Float(A=1, B=$Particles.RibbonLinkOrder)
    Pan Noise=true
    Pan Noise Field=<0 0 2>
   )
   &SkeletalMeshLocation(
    Bone / Socket Index=$Particles.PartitionParticles.Partition
    Bone Sampling Mode=::NewEnumerator4
    MaskSpawnGroup=false
    Sampled Position Offset=&MakeCustomVectorFromBool(
     Bool=&SetBoolByIntComparison(A=$Particles.PartitionParticles.Partition, B=1, Comparison Type=::NewEnumerator0)
     True Vector=<0 0 0>
     False Vector=<0 0 17>
    )
    Sampled Position Offset Coordinate Space:ENiagaraSkelMeshTransforms=::NewEnumerator1
    Skeletal Mesh=$User.SkeletalMesh
   ) {module=SkeletalMeshLocation001}
   /SetVariables(Particles.Position=&AddVectorToPosition(
     Position=$Particles.Position
     Vector=&Multiply_VectorByFloat(
      Vector=$System.DiveVelocity
      Float=&FloatFromCurve(
       FloatCurve=DataInterface() {propertyValues {Curve {keys [{tangentMode=RCTM_User, time=0, value=0}, {tangentMode=RCTM_User, time=1, value=1}]}}}
       CurveIndex=&ReturnNormalizedExecIndex()
       Scale Curve=&MakeCustomFloatFromBool(
        Bool=&SetBoolByIntComparison(A=$Particles.PartitionParticles.Partition, B=1, Comparison Type=::NewEnumerator0)
        True Float=-50
        False Float=-100
       )
      )
     )
    )) {isSetParameters=true, module=SetVariables_#7}
   /SetVariables(Particles.Position=&AddVectorToPosition(
     Position=$Particles.Position
     Vector=&TransformVector(
      Vector=&Multiply_VectorByFloat(
       Vector=&SelectVectorFromArray(
        Direct Array Index=$Particles.PartitionParticles.Partition
        Vector Selection Array=DataInterface() {propertyValues {FloatData [<0.75 0 0>, <-0.75 0 0>, <0.75 0 0>, <-0.75 0 0>]}}
       )
       Float=&FloatFromCurve(
        FloatCurve=DataInterface() {propertyValues {Curve {keys [
            {arriveTangent=4.1785, interpMode=RCIM_Cubic, leaveTangent=4.1785, tangentMode=RCTM_User, time=0, value=0}
            {arriveTangent=1.1179, interpMode=RCIM_Cubic, leaveTangent=1.1179, tangentMode=RCTM_User, time=0.2042, value=0.6228}
            {arriveTangent=0.0224, interpMode=RCIM_Cubic, leaveTangent=0.0224, tangentMode=RCTM_User, time=1, value=1}
           ]}}}
        CurveIndex=&ReturnNormalizedExecIndex()
        Scale Curve=&MakeCustomFloatFromBool(
         Bool=&SetBoolByIntComparison(A=$Particles.PartitionParticles.Partition, B=1, Comparison Type=::NewEnumerator0)
         True Float=-6
         False Float=-13
        )
       )
      )
      Destination Space=::Simulation
     )
    )) {isSetParameters=true, module=SetVariables_#8}
   /SetVariables(Particles.Position=&AddVectorToPosition(
     Position=$Particles.Position
     Vector=&Multiply_VectorByFloat(
      Vector=&NormalizeVector(Vector To Normalize=$Output.CurlNoiseForce.CurlNoiseForce, Safe Normalize=true, Safe Fallback Threshold=0, Safe Fallback Vector=<0 0 0.001>)
      Float=&FloatFromCurve(
       FloatCurve=DataInterface() {propertyValues {Curve {keys [{tangentMode=RCTM_User, time=0, value=0}, {tangentMode=RCTM_User, time=1, value=1}]}}}
       CurveIndex=&ReturnNormalizedExecIndex()
       Scale Curve=&Lerp_Float(A=1, B=2, Alpha=$User.ParaGlideLeanAlpha)
      )
     )
    )) {enabled=false, isSetParameters=true, module=SetVariables_#9}
   &ScaleRibbonWidth(Ribbon Width Scale=1.15)
  ]
  rendererClasses [&NiagaraSpriteRendererProperties, &NiagaraRibbonRendererProperties]
  renderers [
   {rendererClass=&NiagaraSpriteRendererProperties}
   {rendererClass=&NiagaraRibbonRendererProperties, rendererIndex=1}
  ]
 }
 FN_Contrails_Sparkles {
  emitterSpawnScript [&FN_PreserveSkydiveUserVariables() {module=PreserveUserVariables}]
  emitterUpdateScript [
   &EmitterState(
    Enable Distance Culling=true
    Life Cycle Mode=::NewEnumerator1
    Loop Behavior=::NewEnumerator1
    Loop Delay=-0.0033
    Loop Duration Mode=::NewEnumerator1
    MaxDistance=700
    Module.bUseMaxDistance=true
    Module.bUseMinDistance=true
    Scalability Mode=::NewEnumerator1
    UseLoopDelay=false
   )
   &SpawnRate(SpawnRate=&UniformRangedFloat(Minimum=15, Maximum=25), Use Spawn Probability=false)
  ]
  particleSpawnScript [
   /Game/Effects/Niagara/Modules/Groups/FN_SetSpawnGroup_v1/FN_SetSpawnGroup(Emitter Bone Count=4)
   /Niagara/Modules/Spawn/Initialization/InitializeParticle(
    Color=<rgba 2 2 2 1>
    Lifetime=&UniformRangedFloat(Minimum=0.25, Maximum=1)
    Module.Write Color=true
    Module.Write Lifetime=true
    Module.Write Mass=true
    Module.Write Material Random=true
    Module.Write Position=false
    Module.Write Scale=false
    Module.Write SpriteRotation=false
    Module.Write SpriteSize=true
    Position=&SimulationPosition()
    Sprite Size=&UniformRangedVector2D(Minimum=<8 8>, Maximum=<10 10>)
   )
   &SampleSkeletalMeshSkeleton(Mesh:NiagaraDataInterfaceSkeletalMesh=$User.SkeletalMesh, Specific Bone Index:NiagaraInt32=$Particles.SpawnGroup)
   /Niagara/Modules/Spawn/Location/SkeletalMeshSkeletonLocation(Module.Inherit Velocity=false, Module.Use Sampled Bone as Ribbon ID=true)
   /Niagara/Modules/Spawn/Orientation/V2/InitialMeshOrientation(
    Orientation Vector=&UniformRangedVector(Minimum=<-1 -1 -1>, Maximum=<1 1 1>)
    Use Orientation Vector=true
    Use Rotation Vector=false
   )
   /SetVariables(Particles.RotationOrientation:NiagaraFloat=&UniformRangedFloat(Minimum=-1, Maximum=1)) {isSetParameters=true}
   /SetVariables(Particles.RotationSpeed:Vector3f=&UniformRangedVector(Minimum=<-1 -1 -1>, Maximum=<1 1 1>)) {isSetParameters=true}
   /Niagara/Modules/Spawn/Location/SphereLocation(MaskSpawnGroup=false, Module.Override Randomness=false, Sphere Radius=5, Surface Only=false)
  ]
  particleUpdateScript [
   &ParticleState()
   &InheritVelocity(
    Module.Limit Speed=false
    Velocity Scale=&UniformRangedVector(Minimum=<-0.0025 -0.0025 -0.005>, Maximum=<-0.005 -0.005 -0.02>)
   ) {enabled=false}
   &self:DistanceScaleValue(DistanceMax=200, DistanceMin=75) {enabled=false, module=DistanceScaleValue}
   &AddVelocity(
    Scale Added Velocity=<1 1 10>
    Velocity=&FN_FrontEndVector3Modifier(
     Gameplay=&UniformRangedVector(Minimum=<-1 -1 0.5>, Maximum=<1 1 1>)
     Front End=&UniformRangedVector(Minimum=<-1 -1 0.5>, Maximum=<1 1 2>)
    )
   )
   &SolveForcesAndVelocity(Clamp Velocity=false, Limit Acceleration=false)
   &ScaleSpriteSize(Scale Factor=&Multiply_Vector2(
     A=&Vector2DFromFloat(Value=$Emitter.Scalability.DistanceFraction)
     B=&Vector2DFromFloat(Value=&FloatFromCurve(FloatCurve=DataInterface() {propertyValues {Curve {keys [
           {arriveTangent=3.8361, interpMode=RCIM_Cubic, leaveTangent=3.8361, tangentMode=RCTM_User, time=0, value=0}
           {arriveTangent=0.002, interpMode=RCIM_Cubic, leaveTangent=0.002, tangentMode=RCTM_User, time=0.6, value=2}
           <at 1.0011 0.002>
          ]}}}))
    ))
   /SetVariables(Particles.Color=&MultiplyLinearColor(
     Linear Color=&MultiplyLinearColor(Linear Color=$Particles.Color, Scale Factor=&MakeVector4_FromFloat(Float=$Particles.DistanceValue))
     Scale Factor=&MakeVector4(X=1, Y=1, Z=1, W=0)
    )) {enabled=false, isSetParameters=true}
   /SetVariables(Particles.MultColor:LinearColor=&RandomRangeLinearColor(Minimum=<rgba 1 0.3705 0 1>, Maximum=<rgba 0.9729 1 0 1>)) {isSetParameters=true}
   /SetVariables(Particles.GlowColor=&MultiplyLinearColor(
     Linear Color=<rgba 1 0.4733 0 1>
     Scale Factor=&MakeVector4_FromLinearColor(LinearColor=$Particles.MultColor)
    )) {isSetParameters=true}
   /SetVariables(Particles.GlowSize=&Multiply_Vector2(A=$Particles.SpriteSize, B=<2 2>)) {isSetParameters=true}
  ]
  rendererClasses [&NiagaraSpriteRendererProperties]
  renderers [
   {rendererClass=&NiagaraSpriteRendererProperties}
   {rendererClass=&NiagaraSpriteRendererProperties, rendererIndex=1}
  ]
 }
 PresentRibbon001 {
  bEnabled=false
  emitterUpdateScript [
   &EmitterState()
   &FN_SpawnRibbon_v2(Number of Ribbons=4, Spawn Mode=::NewEnumerator0, SpawnRate=100)
  ]
  particleSpawnScript [
   &FN_InitializeRibbon_v3(
    Color:LinearColor=<rgba 1 1 1 2>
    Enable Skeletal Position=true
    Offset Axis:ENiagaraOrientationAxis=::ZAxis
    Ribbon Width=20
    Sampled Position Coordinate Space:ENiagaraCoordinateSpace=::Local
    Sampled Rotation Coordinate Space:ENiagaraCoordinateSpace=::Local
    Sampled Velocity Coordinate Space:ENiagaraCoordinateSpace=::Local
   ) {module=FN_InitializeRibbon_v3001}
   &SkeletalMeshLocation(
    Apply Sampled Orientation to Mesh Particles=true
    Bone / Socket Index=$Particles.SpawnGroup
    Bone Sampling Mode=::NewEnumerator4
    Inherit Velocity=false
    Interpolate Mesh Sampling=false
    MaskSpawnGroup=false
    Sampled Position Offset=&MakeCustomVectorFromBool(
     Bool=&SetBoolByIntComparison(A=$Particles.SpawnGroup, B=1, Comparison Type=::NewEnumerator0)
     True Vector=<0 0 0>
     False Vector=<0 0 15>
    )
    Sampled Position Offset Coordinate Space:ENiagaraMeshTransforms=::NewEnumerator1
    Skeletal Mesh=$User.SkeletalMesh
   )
   &AddVelocity(
    Scale Added Velocity=&Lerp_Vector(A=<0.5 0.5 0.5>, B=<0.5 0.5 0.5>, Alpha=$Emitter.Dive)
    Velocity=&FN_FrontEndVector3Modifier(
     Gameplay=&MakeCustomVectorFromBool(
      Bool=&SetBoolByIntComparison(A=&SpawnGroup(), B=1, Comparison Type=::NewEnumerator0)
      True Vector=<0 0 35>
      False Vector=<0 0 350>
     )
     Front End=&MakeCustomVectorFromBool(
      Bool=&SetBoolByIntComparison(A=&SpawnGroup(), B=1, Comparison Type=::NewEnumerator0)
      True Vector=&Lerp_Vector(A=<0 0 50>, B=<0 -200 50>, Alpha=$User.ParaGlideLeanAlpha)
      False Vector=&Lerp_Vector(A=<0 0 300>, B=<0 -200 300>, Alpha=$User.ParaGlideLeanAlpha)
     )
    )
   )
   &AddVelocity(
    Velocity=&SelectVectorFromArray(
     Direct Array Index=$Particles.SpawnGroup
     Vector Selection Array=DataInterface() {propertyValues {FloatData [<-1 0 0>, <1 0 0>, <-1 0 0>, <1 0 0>]}}
    )
    Velocity Speed Scale=&MakeCustomFloatFromBool(
     Bool=&SetBoolByIntComparison(A=$Particles.SpawnGroup, B=1, Comparison Type=::NewEnumerator0)
     True Float=15
     False Float=25
    )
   ) {module=AddVelocity001}
   &DynamicMaterialParameters(
    Index 0 Param 1=&RandomRangeFloat(Minimum=0, Maximum=1)
    Index 0 Param 2=1
    Index 0 Param 3=1
    Index 0 Param 4=&RandomRangeFloat(Minimum=0, Maximum=1)
    Module.Param0WriteEnabled=true
    Module.Param1WriteEnabled=true
    Module.Param2WriteEnabled=true
    Module.Param3WriteEnabled=true
   ) {enabled=false}
   &DynamicMaterialParameters(
    Index 0 Param 1=&Lerp_Float(A=0, B=0.2, Alpha=&MakeFloatFromInt(Int=$Particles.SpawnGroup))
    Index 0 Param 2=1.15
    Index 0 Param 3=1
    Index 0 Param 4=&Lerp_Float(A=-0.15, B=0.15, Alpha=&MakeFloatFromInt(Int=$Particles.SpawnGroup))
    Module.Param0WriteEnabled=true
    Module.Param1WriteEnabled=true
    Module.Param2WriteEnabled=true
    Module.Param3WriteEnabled=true
   ) {module=DynamicMaterialParameters001}
  ]
  particleUpdateScript [
   &CurlNoiseForce(
    Noise Frequency=25
    Noise Strength=&Multiply_Float(
     A=&Lerp_Float(A=0.15, B=0.6, Alpha=$User.ParaGlideLeanAlpha)
     B=&MakeCustomFloatFromBool(
      Bool=&SetBoolByIntComparison(A=$Particles.SpawnGroup, B=1, Comparison Type=::NewEnumerator0)
      True Float=100
      False Float=250
     )
    )
    Pan Noise=true
    Pan Noise Field=<0 0 5>
   )
   &ScaleColor(
    Scale Alpha=&FloatFromCurve(FloatCurve=DataInterface() {propertyValues {Curve {keys [<at 0.85 1>,<at 0.95 0>]}}})
    ScaleA=true
    ScaleRGB=true
   )
   &Drag(
    Drag=&MakeCustomFloatFromBool(
     Bool=&SetBoolByIntComparison(A=$Particles.SpawnGroup, B=1, Comparison Type=::NewEnumerator0)
     True Float=1
     False Float=2
    )
    Use Linear Drag=true
    Use Rotational Drag=true
   )
   &GravityForce(Coordinate Space=::Local, Gravity=<0 0 500>)
   &SolveForcesAndVelocity(Clamp Velocity=false, Limit Acceleration=false)
   &FN_UpdateRibbon(
    Enable Skeletal Position=true
    Initial Color:LinearColor=$Particles.Initial.Color
    Initial RibbonWidth:NiagaraFloat=$Particles.Initial.RibbonWidth
   )
   &ScaleRibbonWidth(Ribbon Width Scale=&FloatFromCurve(FloatCurve=DataInterface() {propertyValues {Curve {keys [{tangentMode=RCTM_User, time=0, value=1}, {tangentMode=RCTM_User, time=1, value=2}]}}}))
  ]
  rendererClasses [&NiagaraRibbonRendererProperties]
  renderers [{rendererClass=&NiagaraRibbonRendererProperties}]
 }
 PresentRibbon002 {
  bEnabled=false
  emitterUpdateScript [
   &EmitterState()
   &FN_SpawnRibbon_v2(Number of Ribbons=4, Spawn Mode=::NewEnumerator0, SpawnRate=100)
  ]
  particleSpawnScript [
   &FN_InitializeRibbon_v3(
    Color:LinearColor=<rgba 1 1 1 2>
    Enable Skeletal Position=true
    Offset Axis:ENiagaraOrientationAxis=::ZAxis
    Ribbon Width=15
    Sampled Position Coordinate Space:ENiagaraCoordinateSpace=::Local
    Sampled Rotation Coordinate Space:ENiagaraCoordinateSpace=::Local
    Sampled Velocity Coordinate Space:ENiagaraCoordinateSpace=::Local
   ) {module=FN_InitializeRibbon_v3001}
   &SkeletalMeshLocation(
    Apply Sampled Orientation to Mesh Particles=true
    Bone / Socket Index=$Particles.SpawnGroup
    Bone Sampling Mode=::NewEnumerator4
    Inherit Velocity=false
    Interpolate Mesh Sampling=false
    MaskSpawnGroup=false
    Sampled Position Offset=&MakeCustomVectorFromBool(
     Bool=&SetBoolByIntComparison(A=$Particles.SpawnGroup, B=1, Comparison Type=::NewEnumerator0)
     True Vector=<0 0 0>
     False Vector=<0 0 15>
    )
    Sampled Position Offset Coordinate Space:ENiagaraMeshTransforms=::NewEnumerator1
    Skeletal Mesh=$User.SkeletalMesh
   )
   &AddVelocity(
    Scale Added Velocity=&Lerp_Vector(A=<0.5 0.5 0.5>, B=<0.5 0.5 0.5>, Alpha=$Emitter.Dive)
    Velocity=&FN_FrontEndVector3Modifier(
     Gameplay=&MakeCustomVectorFromBool(
      Bool=&SetBoolByIntComparison(A=&SpawnGroup(), B=1, Comparison Type=::NewEnumerator0)
      True Vector=<0 0 50>
      False Vector=<0 0 650>
     )
     Front End=&MakeCustomVectorFromBool(
      Bool=&SetBoolByIntComparison(A=&SpawnGroup(), B=1, Comparison Type=::NewEnumerator0)
      True Vector=&Lerp_Vector(A=<0 0 50>, B=<0 -200 50>, Alpha=$User.ParaGlideLeanAlpha)
      False Vector=&Lerp_Vector(A=<0 0 300>, B=<0 -200 300>, Alpha=$User.ParaGlideLeanAlpha)
     )
    )
   )
   &AddVelocity(
    Velocity=&SelectVectorFromArray(
     Direct Array Index=$Particles.SpawnGroup
     Vector Selection Array=DataInterface() {propertyValues {FloatData [<1 0 0>, <-1 0 0>, <1 0 0>, <-1 0 0>]}}
    )
    Velocity Speed Scale=&MakeCustomFloatFromBool(
     Bool=&SetBoolByIntComparison(A=$Particles.SpawnGroup, B=1, Comparison Type=::NewEnumerator0)
     True Float=15
     False Float=25
    )
   ) {module=AddVelocity001}
   &DynamicMaterialParameters(
    Index 0 Param 1=&RandomRangeFloat(Minimum=0, Maximum=1)
    Index 0 Param 2=1
    Index 0 Param 3=1
    Index 0 Param 4=&RandomRangeFloat(Minimum=0, Maximum=1)
    Module.Param0WriteEnabled=true
    Module.Param1WriteEnabled=true
    Module.Param2WriteEnabled=true
    Module.Param3WriteEnabled=true
   ) {enabled=false}
   &DynamicMaterialParameters(
    Index 0 Param 1=&Lerp_Float(A=0, B=0.2, Alpha=&MakeFloatFromInt(Int=$Particles.SpawnGroup))
    Index 0 Param 2=1.15
    Index 0 Param 3=1
    Index 0 Param 4=&Lerp_Float(A=-0.15, B=0.15, Alpha=&MakeFloatFromInt(Int=$Particles.SpawnGroup))
    Module.Param0WriteEnabled=true
    Module.Param1WriteEnabled=true
    Module.Param2WriteEnabled=true
    Module.Param3WriteEnabled=true
   ) {module=DynamicMaterialParameters001}
  ]
  particleUpdateScript [
   &CurlNoiseForce(
    Noise Frequency=25
    Noise Strength=&Multiply_Float(
     A=&Lerp_Float(A=0.15, B=0.55, Alpha=$User.ParaGlideLeanAlpha)
     B=&MakeCustomFloatFromBool(
      Bool=&SetBoolByIntComparison(A=$Particles.SpawnGroup, B=1, Comparison Type=::NewEnumerator0)
      True Float=100
      False Float=250
     )
    )
    Pan Noise=true
    Pan Noise Field=<0 0 5>
   )
   &ScaleColor(
    Scale Alpha=&FloatFromCurve(FloatCurve=DataInterface() {propertyValues {Curve {keys [<at 0.85 1>,<at 0.95 0>]}}})
    ScaleA=true
    ScaleRGB=true
   )
   &Drag(Drag=3, Rotational Drag=0, Use Linear Drag=true, Use Rotational Drag=true)
   &GravityForce(Coordinate Space=::Local, Gravity=<0 0 500>)
   &SolveForcesAndVelocity(Clamp Velocity=false, Limit Acceleration=false)
   &FN_UpdateRibbon(
    Enable Skeletal Position=true
    Initial Color:LinearColor=$Particles.Initial.Color
    Initial RibbonWidth:NiagaraFloat=$Particles.Initial.RibbonWidth
   )
   &ScaleRibbonWidth(Ribbon Width Scale=&FloatFromCurve(FloatCurve=DataInterface() {propertyValues {Curve {keys [{tangentMode=RCTM_User, time=0, value=1}, {tangentMode=RCTM_User, time=0.5, value=2}]}}}))
  ]
  rendererClasses [&NiagaraRibbonRendererProperties]
  renderers [{rendererClass=&NiagaraRibbonRendererProperties}]
 }
 Trails {
  emitterUpdateScript [
   &EmitterState()
   &SpawnBurst_Instantaneous(Spawn Count=4, Use Spawn Probability=false, UseLoopCountLimit=true)
  ]
  particleSpawnScript [
   /Game/Effects/Niagara/Modules/Groups/FN_SetSpawnGroup()
   /Niagara/Modules/Spawn/Initialization/V2/InitializeParticle(Color=<rgba 1 1 1 0.35>, Mesh Scale=<0.45 0.45 3>, Mesh Scale Mode=::NewEnumerator4, UsePositionOffset=false, Write Material Random=true)
   /Niagara/Modules/Spawn/Orientation/V3/InitialMeshOrientation(Mesh Axis To Orient=<0 0 -1>, Mesh Orientation Mode=::NewEnumerator2, Use Rotation Vector=false)
  ]
  particleUpdateScript [
   &ParticleState(Kill Particles When Lifetime Has Elapsed=false, Let Infinitely Lived Particles Die When Emitter Deactivates=true)
   /Niagara/Modules/Update/Orientation/V3/UpdateMeshOrientation(
    Facing Direction=$System.DiveVelocity
    Facing Interpolation=false
    Orientation Method=::NewEnumerator1
    Override Quat=false
    Side Direction:Vector3f=$Engine.Owner.SystemXAxis
    Stabilization Mode:ENiagara_OrientVectorCount=::NewEnumerator2
    TransformFacing=false
    TransformSide=false
   ) {module=UpdateMeshOrientation002}
   &SkeletalMeshLocation(
    Bone / Socket Index=$Particles.SpawnGroup
    Bone Sampling Mode=::NewEnumerator4
    Inherit Velocity=false
    MaskSpawnGroup=false
    Sampled Position Offset=<0 0 7>
    Sampled Position Offset Coordinate Space:ENiagaraMeshTransforms=::NewEnumerator1
    Skeletal Mesh=$User.SkeletalMesh
   ) {enabled=false, module=SkeletalMeshLocation001}
   &SkeletalMeshLocation(
    Bone / Socket Index=$Particles.SpawnGroup
    Bone Sampling Mode=::NewEnumerator4
    MaskSpawnGroup=false
    Orientation Sampling=::NewEnumerator1
    Sampled Position Offset=&MakeCustomVectorFromBool(
     Bool=&SetBoolByIntComparison(A=$Particles.SpawnGroup, B=1, Comparison Type=::NewEnumerator0)
     True Vector=<0 0 -28>
     False Vector=<1 0 -5>
    )
    Sampled Position Offset Coordinate Space:ENiagaraSkelMeshTransforms=::NewEnumerator1
    Skeletal Mesh=$User.SkeletalMesh
   )
   &InterpolateOverTime(Rate Of Change=10, Target Vector Value:NiagaraPosition=$Particles.Position, Variable Type=::NewEnumerator2)
   /SetVariables(Particles.InterpolateDynamic:Vector4f=&MakeFloat4FromVectorAndFloat("Vector (RGB)"=$Output.InterpolateOverTime.Moving Average, "Float (Alpha)"=1)) {enabled=false, isSetParameters=true}
   &SolveForcesAndVelocity(Clamp Velocity=false, Limit Acceleration=false)
   &DynamicMaterialParameters(
    Index 0 Param 1=1
    Index 0 Param 3=&Multiply_Float(A=5, B=&Lerp_Float(A=5, B=10, Alpha=$System.Dive))
    Index 2 Param 1=&Lerp_Float(A=0.1, B=0.5, Alpha=&Ceil_Float(Float=$System.Dive))
    Index 2 Param 2=&Lerp_Float(A=0, B=0.5, Alpha=&Ceil_Float(Float=$System.Dive))
    Module.Param0WriteEnabled=true
    Module.Param0WriteEnabled2=true
    Module.Param1WriteEnabled=false
    Module.Param1WriteEnabled2=true
    Module.Param2WriteEnabled=true
    Module.Param2WriteEnabled2=false
    Module.Param3WriteEnabled=false
    Module.Param3WriteEnabled2=false
    Write Parameter Index 2=true
   )
   &ScaleColor(Scale Alpha=0.6, ScaleA=true, ScaleRGB=true)
   &ScaleMeshSize(Scale Factor=&MakeCustomVectorFromBool(
     Bool=&SetBoolByIntComparison(A=$Particles.SpawnGroup, B=1, Comparison Type=::NewEnumerator0)
     True Vector=&Lerp_Vector(A=<1 1 0.25>, B=<1 1 0.35>, Alpha=$System.Dive)
     False Vector=&Lerp_Vector(A=<1 1 0.25>, B=<1 1 0.4>, Alpha=$System.Dive)
    ))
   &ScaleMeshSize(Scale Factor=&Multiply_VectorByFloat(Vector=<1.1 1.1 1>, Float=1)) {enabled=false, module=ScaleMeshSize001}
  ]
  rendererClasses [&NiagaraMeshRendererProperties]
  renderers [{rendererClass=&NiagaraMeshRendererProperties, rendererIndex=0}]
 }
 TreeMesh {
  emitterUpdateScript [
   &EmitterState(
    Inactive Response=::NewEnumerator2
    Life Cycle Mode=::NewEnumerator1
    Loop Behavior=::NewEnumerator1
    Loop Duration Mode=::NewEnumerator1
    UseLoopDelay=false
   )
   &SpawnBurst_Instantaneous(Spawn Count=2, Use Spawn Probability=false, UseLoopCountLimit=false) {module=SpawnBurst_Instantaneous001}
  ]
  particleSpawnScript [
   &FN_SetSpawnGroup(Number of Spawn Groups=2)
   &InitializeParticle(Mesh Scale=<0.25 1.15 0.5>, Mesh Scale Mode=::NewEnumerator4, UsePositionOffset=false, Write Material Random=true) {module=InitializeParticle001}
   &CylinderLocation(
    Cylinder Height=1
    Cylinder Radius=1
    Lathe Profile=DataInterface() {propertyValues {Curve {keys []}}}
    MaskSpawnGroup=false
    Module.Enable Lathe Profile=false
    Module.Override Randomness=false
    Module.Surface Only=false
   ) {enabled=false}
   &AddVelocity(Velocity=&UniformRangedVector(
     Minimum=&MakeVector(
      X=0
      Y=0
      Z=&Multiply_Float(
       A=150
       B=&FloatFromCurve(
        FloatCurve=DataInterface() {propertyValues {Curve {keys [<at 0 0.9>,<at 1 2>]}}}
        CurveIndex=&Divide_Float(A=&VectorLength(Vector=$Engine.Owner.Velocity), B=9000)
       )
      )
     )
     Maximum=&MakeVector(
      X=0
      Y=0
      Z=&Multiply_Float(
       A=200
       B=&FloatFromCurve(
        FloatCurve=DataInterface() {propertyValues {Curve {keys [<at 0 0.9>,<at 1 2>]}}}
        CurveIndex=&Divide_Float(A=&VectorLength(Vector=$Engine.Owner.Velocity), B=9000)
       )
      )
     )
    )) {enabled=false}
   /Niagara/Modules/Spawn/Orientation/V2/InitialMeshOrientation(Use Orientation Vector=true, Use Rotation Vector=false) {enabled=false}
   /SetVariables(Particles.RotationOrientation:NiagaraFloat=&UniformRangedFloat(Minimum=-1, Maximum=1)) {enabled=false, isSetParameters=true, module=SetVariables_#4}
   /SetVariables(Particles.RotationSpeed:Vector3f=&UniformRangedVector(Minimum=<0 0 -0.5>, Maximum=<0 0 0.5>)) {enabled=false, isSetParameters=true, module=SetVariables_#5}
   /Niagara/Modules/Spawn/Orientation/V3/InitialMeshOrientation(Mesh Axis To Orient=<0 0 -1>, Mesh Orientation Mode=::NewEnumerator2, Use Rotation Vector=false) {module=InitialMeshOrientation001}
  ]
  particleUpdateScript [
   &ParticleState(Kill Particles When Lifetime Has Elapsed=false, Loop Particles Lifetime=true)
   &UpdateMeshOrientation(
    Rotation Rate=&FloatFromCurve(
     FloatCurve=DataInterface() {propertyValues {Curve {keys [{tangentMode=RCTM_User, time=0, value=0}, {tangentMode=RCTM_User, time=0.5, value=1}]}}}
     CurveIndex=$Emitter.Age
     Scale Curve=&Lerp_Float(A=0.25, B=2, Alpha=$User.ParaGlideLeanAlpha)
    )
    Rotation Vector=<0 0 1>
   ) {module=UpdateMeshOrientation001}
   &UpdateMeshOrientation(
    Facing Direction=$System.DiveVelocity
    Facing Interpolation=false
    Orientation Method=::NewEnumerator1
    Override Quat=false
    Side Direction:Vector3f=$Engine.Owner.SystemXAxis
    Stabilization Mode:ENiagara_OrientVectorCount=::NewEnumerator2
    TransformFacing=false
    TransformSide=false
   ) {module=UpdateMeshOrientation002}
   &SkeletalMeshLocation(
    Bone / Socket Index=$Particles.SpawnGroup
    Bone Sampling Mode=::NewEnumerator4
    Inherit Velocity=false
    MaskSpawnGroup=false
    Sampled Position Offset=<0 0 20>
    Sampled Position Offset Coordinate Space:ENiagaraMeshTransforms=::NewEnumerator1
    Skeletal Mesh=$User.SkeletalMesh
   )
   /Niagara/Modules/Update/Orientation/V2/UpdateMeshOrientation(Rotation Rate:NiagaraFloat=$Particles.RotationOrientation, Rotation Vector:Vector3f=$Particles.RotationSpeed) {enabled=false}
   &Drag(Drag=0.3, Use Linear Drag=true, Use Rotational Drag=false)
   /Niagara/Modules/Update/Forces/CurlNoiseForce(
    Noise Frequency:NiagaraFloat=&UniformRangedFloat(Minimum=0.1, Maximum=1)
    Noise Strength=&UniformRangedFloat(Minimum=300, Maximum=600)
    Pan Noise=false
    Pan Noise Field=<1 1 0>
   ) {enabled=false}
   &DynamicMaterialParameters(
    Index 0 Param 1=&FloatFromCurve(FloatCurve=DataInterface() {propertyValues {Curve {keys [<at -0.0046 2>,<at 1.0011 0.002>]}}})
    Module.Param0WriteEnabled=true
    Module.Param1WriteEnabled=true
    Module.Param2WriteEnabled=true
    Module.Param3WriteEnabled=true
   ) {enabled=false}
   &ScaleMeshSize(Scale Factor=&Multiply_Vector(
     A=<1 1 1>
     B=&VectorFromFloat(Value=&FloatFromCurve(FloatCurve=DataInterface() {propertyValues {Curve {keys [
           <at -0.0046 1>
           <at 0.3771 0.7748>
           {arriveTangent=-0.3128, leaveTangent=-0.3128, time=0.4459, value=1.2326}
           <at 0.5037 0.6933>
           <at 0.5587 0.6619>
           <at 0.611 0.8501>
           <at 0.6536 0.618>
           <at 0.7155 0.6055>
           <at 0.8338 0.5741>
           <at 1.0011 0.5>
          ]}}}))
    )) {enabled=false}
   &SolveForcesAndVelocity(Clamp Velocity=false, Limit Acceleration=false)
   &GenerateLocationEvent(Event Send Rate:NiagaraFloat=&UniformRangedFloat(Minimum=1, Maximum=3), Module.Use Event Send Rate=true) {enabled=false}
  ]
  rendererClasses [&NiagaraMeshRendererProperties]
  renderers [{rendererClass=&NiagaraMeshRendererProperties, rendererIndex=0}]
 }
}
summary {
 systemName=NS_Contrail_JollyTroll
 emitters [
  {emitterName=FN_Contrails_Sparkles, rendererClasses [&NiagaraSpriteRendererProperties]}
  {emitterName=TreeMesh, rendererClasses [&NiagaraMeshRendererProperties]}
  {emitterName=BoxMesh, rendererClasses [&NiagaraMeshRendererProperties]}
  {bEnabled=false, emitterName=PresentRibbon001, rendererClasses [&NiagaraRibbonRendererProperties]}
  {emitterName=Trails, rendererClasses [&NiagaraMeshRendererProperties]}
  {emitterName=ChristmasDecorationsHands, rendererClasses [&NiagaraSpriteRendererProperties]}
  {bEnabled=false, emitterName=BowRibbons, rendererClasses [&NiagaraMeshRendererProperties]}
  {bEnabled=false, emitterName=PresentRibbon002, rendererClasses [&NiagaraRibbonRendererProperties]}
  {emitterName=BoxGlow, rendererClasses [&NiagaraSpriteRendererProperties]}
  {emitterName=Empty, rendererClasses [&NiagaraSpriteRendererProperties, &NiagaraRibbonRendererProperties]}
  {emitterName=Empty001, rendererClasses [&NiagaraSpriteRendererProperties, &NiagaraRibbonRendererProperties]}
 ]
 userVariables [
  User.bIsFrontEnd {type=NiagaraBool}
  User.bIsFrontEndPreview {type=NiagaraBool, defaultValue=true}
  User.ParaGlideLeanAlpha {type=NiagaraFloat, defaultValue=0, description=""}
  User.Player {
   type=NiagaraDataInterfaceSkeletalMesh
   defaultValue {
    struct=NiagaraExt_VariableValue_DataInterface
    value {dataInterface=&self:NiagaraDataInterfaceSkeletalMesh_2}
   }
  }
  User.PlayerSpeed {type=NiagaraFloat, defaultValue=0, description=""}
  User.RibbonIDOffset {type=NiagaraInt32}
  User.RibbonSpawnRate {type=NiagaraFloat, defaultValue=40, description=""}
  User.SkeletalMesh {
   type=NiagaraDataInterfaceSkeletalMesh
   defaultValue {
    struct=NiagaraExt_VariableValue_DataInterface
    value {dataInterface=&self:NiagaraDataInterfaceSkeletalMesh_0}
   }
  }
  User.SkeletalMesh2 {
   type=NiagaraDataInterfaceSkeletalMesh
   defaultValue {
    struct=NiagaraExt_VariableValue_DataInterface
    value {dataInterface=&self:NiagaraDataInterfaceSkeletalMesh_1}
   }
  }
  User.TrailAlpha {type=NiagaraFloat, defaultValue=1, description=""}
  User.V2 {type=NiagaraFloat, defaultValue=0, description=""}
 ]
}
